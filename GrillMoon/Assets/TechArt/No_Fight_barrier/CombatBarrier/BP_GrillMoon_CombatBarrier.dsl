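; GrillMoon combat barrier actor: a zone that switches players between event-safe and combat modes.
; Calendar events, day events and cheat events toggle its collision and visibility.
; Router combat-zone events add combat aura and default gravity; ending overlap restores event aura, damage immunity and gravity, and strips event items.
; A gated looping timer damages downed combatants; a DoOnce guards router listening against soft locks.
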(event EventBeginPlay
  ;[0.4927 0.2838 0.5469 0.5] Begin Play / Destroy
  (Networking|SwitchHasAuthority
    (:Authority
      (Class|FortAsyncActionAuthorityMatchReady|MatchReadyAsync
        (:OnReady
          (MCPContext|InitialCalendarSyncCompleted
            (:OnSyncComplete
              (CallFunction|WatchCalendarEvents)
              (CallFunction|EventRouter-StartListening))))))))

(event Custom|GM_CZoneOn
  ;[0.03969 0.2823 0.5208 0.5] Cheat Custom Events
  (CallFunction|SetZoneActive true))

(event Custom|Watch%20Calendar%20Events
  ;[0.1835 0.5208 0.3761 0.5] Calendar Event
  (MCPContext|WatchCalendarEvent (Variables|Default|GetEventZoneCalendarEventName)
    (:OnEventActive
      (CallFunction|SetZoneActive true) ; Set zone active
    )
    (:OnEventEnded
      (CallFunction|SetZoneActive false) ; Disable zone
    )))

(event Custom|GM_CZoneOff
  ;[0.03969 0.2823 0.5208 0.5] Cheat Custom Events
  (CallFunction|SetZoneActive false))

(event Custom|Clear_Zone
  ;[0.4927 0.2838 0.5469 0.5] Begin Play / Destroy
  (Actor|DestroyActor))

(event OnComponentBeginOverlap(CombatZone)
  (params (OverlappedComponent :type Primitive%20Component%20Object%20Reference) (OtherActor :type Actor%20Object%20Reference) (OtherComp :type Primitive%20Component%20Object%20Reference) (OtherBodyIndex :type Integer) (bFromSweep :type Boolean) (SweepResult :type Hit%20Result%20Structure%20%28by%20ref%29))
  ;[0.2534 0.4854 0.5938 0.5] Add / Remove Combat Mode from Players
  (bind _asplayer_pawn_athena (Utilities|Casting|CastToPlayerPawn_Athena OtherActor))
  (if (== (Pawn|GetController _asplayer_pawn_athena) (Game|GetPlayerController 0))
    _))

(event OnComponentEndOverlap(CombatZone)
  (params (OverlappedComponent :type Primitive%20Component%20Object%20Reference) (OtherActor :type Actor%20Object%20Reference) (OtherComp :type Primitive%20Component%20Object%20Reference) (OtherBodyIndex :type Integer))
  ;[0.2534 0.4854 0.5938 0.5] Add / Remove Combat Mode from Players
  (bind _asplayer_pawn_athena (Utilities|Casting|CastToPlayerPawn_Athena OtherActor)
    (:then
      (if (Utilities|Array|ContainsItem (Variables|Default|GetActors%20In%20Combat%20Mode) OtherActor)
        (CallFunction|RemoveCombatMode OtherActor))
      (if (== (Pawn|GetController _asplayer_pawn_athena) (Game|GetPlayerController 0))
        (Rendering|Material|SetScalarParameterValue "/GrillMoon_Countdown/Assets/FX/MPC_GrillMoon.MPC_GrillMoon" "Post_time" (Utilities|Time|GetGameTimeinSeconds) :!class /Script/Engine.KismetMaterialLibrary)
        (Development|PrintString "Combat Barrier - Player Pawn Found" false true "(R=0.000000,G=0.660000,B=1.000000,A=1.000000)" 1e+01)))
    (:CastFailed
      (Development|PrintString (Utilities|String|Append "Combat Barrier - Overlap ended by " (Utilities|GetDisplayName OtherActor)) false true "(R=0.000000,G=0.660000,B=1.000000,A=1.000000)" 2e+01))))

(event Custom|Event%20Router%20-%20Start%20Listening
  ;[0 0.5208 0.5111 0.5] Event Router
  (Utilities|FlowControl|DoOnce false
    (:Completed
      (Utilities|FlowControl|Sequence
        (:then_0
          ;; Day 11 Events
          (bind (_asynctaskproxy _payload) (GameplayEventRouter|StartListeningToStatefulEvent (GameplayEventRouter|GetGlobalEventRouter) "/GrillMoon_Countdown/Gameplay/EventRouter/GrillMoon_Day_Event.GrillMoon_Day_Event")
            (:OnEventReceived
              _)
            (:OnSavedEventStateExists
              _))
          (switch Utilities|FlowControl|Switch|SwitchonGrillMoon_Day_Event_Enum (Utilities|Struct|BreakGrillMoonDayEvent _payload)
            (:NewEnumerator12
              (CallFunction|SetZoneActive true) ; turn on when the event zone is active
            )
            (:NewEnumerator13
              (CallFunction|SetZoneActive false) ; turn off when the event starts
              (CallFunction|StopDamagingDBNOPlayers))))
        (:then_1
          ;; Add / Remove GA for Players
          (bind (_asynctaskproxy_2 _payload_2) (GameplayEventRouter|StartListeningToStatefulEvent (GameplayEventRouter|GetGlobalEventRouter) "/GrillMoon_Countdown/Gameplay/EventRouter/GrillMoon_CombatZoneEvent.GrillMoon_CombatZoneEvent")
            (:OnEventReceived
              (CallFunction|AddCombatModeEffecttoPlayer (Utilities|Struct|BreakGrillMoonCombatZoneEvent _payload_2)))) ; Listen for Player Enter/Exit Events
        )))) ; Listening to Event Routers more than\nonce causes a soft lock. Do not remove\nthis.
)

(event Custom|Auto%20Eliminate%20DBNO%20Player
  (params (Player :type Fort%20Player%20Pawn%20Object%20Reference))
  ;[0.2534 0.4854 0.5938 0.5] Add / Remove Combat Mode from Players
  ;; DBNO Auto-Eliminate
  (EventDispatchers|AssignOnDBNOStateChanged Player (delegate Custom|OnDBNOStateChanged_Event))
  (Utilities|FlowControl|DoOnce false
    (:Completed
      (CallFunction|StartDamagingDBNOPlayers))))

(event Custom|OnDBNOStateChanged_Event
  (params (FortPawn :type Fort%20Pawn%20Object%20Reference) (bInIsDBNO :type Boolean))
  ;[0.2534 0.4854 0.5938 0.5] Add / Remove Combat Mode from Players
  ;; DBNO Auto-Eliminate
  (if bInIsDBNO
    (Utilities|Array|AddUnique (Variables|Default|GetDBNO%20Player%20List) FortPawn)))

(event Custom|Start%20Damaging%20DBNO%20Players
  ;[0.2534 0.4854 0.5938 0.5] Add / Remove Combat Mode from Players
  ;; DBNO Auto-Eliminate
  (Utilities|Time|SetTimerbyEvent (delegate Custom|Damage%20DBNO%20Players) 2.0 true))

(event Custom|GM_SetDBNODamage
  (params (damage :type Float%20%28double-precision%29))
  ;[0.2534 0.4854 0.5938 0.5] Add / Remove Combat Mode from Players
  ;; DBNO Auto-Eliminate
  (Variables|Default|SetDBNO%20Damage%20Amount damage))

(event Custom|Remove%20Player%20from%20DBNO%20List
  (params (Player :type Fort%20Player%20Pawn%20Object%20Reference))
  ;[0.2534 0.4854 0.5938 0.5] Add / Remove Combat Mode from Players
  ;; DBNO Auto-Eliminate
  (EventDispatchers|UnbindEventfromOnDBNOStateChanged Player (delegate Custom|OnDBNOStateChanged_Event)))

(event Custom|Remove%20Combat%20Mode%20from%20All%20Players)

(event Custom|Damage%20DBNO%20Players
  (bind _dbno_player_index (Variables|Default|GetDBNO%20Player%20Index))
  (bind _dbno_player_list (Variables|Default|GetDBNO%20Player%20List))
  (bind _output (Utilities|Array|Get(acopy) _dbno_player_list _dbno_player_index))
  ;[0.2534 0.4854 0.5938 0.5] Add / Remove Combat Mode from Players
  ;; DBNO Auto-Eliminate
  (Utilities|FlowControl|Gate false
    (:Exit
      (while (<= (Variables|Default|GetDBNO%20Player%20Index) (Utilities|Array|LastIndex (Variables|Default|GetDBNO%20Player%20List)))
        (Utilities|IsValid _output
          (:Is%20Valid
            (if (Health|IsDBNO _output)
              (Game|Damage|ApplyDamage _output (Variables|Default|GetDBNO%20Damage%20Amount))))
          (:Is%20Not%20Valid
            _)) ; player valid?
        (Math|Integer|IncrementInt (Variables|Default|GetDBNO%20Player%20Index)))
      (Variables|Default|SetDBNO%20Player%20Index 0))))

(event Custom|Stop%20Damaging%20DBNO%20Players
  (bind _dbno_player_index (Variables|Default|GetDBNO%20Player%20Index))
  (bind _dbno_player_list (Variables|Default|GetDBNO%20Player%20List))
  (bind _output (Utilities|Array|Get(acopy) _dbno_player_list _dbno_player_index))
  ;[0.2534 0.4854 0.5938 0.5] Add / Remove Combat Mode from Players
  ;; DBNO Auto-Eliminate
  (Utilities|FlowControl|Gate false
    (:Exit
      (while (<= (Variables|Default|GetDBNO%20Player%20Index) (Utilities|Array|LastIndex (Variables|Default|GetDBNO%20Player%20List)))
        (Utilities|IsValid _output
          (:Is%20Valid
            (if (Health|IsDBNO _output)
              (Game|Damage|ApplyDamage _output (Variables|Default|GetDBNO%20Damage%20Amount))))
          (:Is%20Not%20Valid
            _)) ; player valid?
        (Math|Integer|IncrementInt (Variables|Default|GetDBNO%20Player%20Index)))
      (Variables|Default|SetDBNO%20Player%20Index 0))))

(event Custom|PlayerAddedToCombatMode)

(fn AddCombatModeEffectToPlayer
  (params (ActorEnteringCombatMode :type Actor%20Object%20Reference))
  (bind _returnvalue (Ability|GetFortAbilitySystemComponent ActorEnteringCombatMode))
  (bind _asfort_player_pawn_athena (Utilities|Casting|CastToFortPlayerPawnAthena ActorEnteringCombatMode))
  (Utilities|IsValid _returnvalue
    (:Is%20Valid
      (FortPawn|SetGravityMultiplier _asfort_player_pawn_athena (Variables|Default|GetDefaultGravityScale))
      (Utilities|IsValid _returnvalue
        (:Is%20Valid
          (GameplayEffects|RemoveActiveGameplayEffectbySourceEffect _returnvalue "/GrillMoon/Gameplay/Effects/GE_GrillMoon_EventAuraFX.GE_GrillMoon_EventAuraFX_C" 0 -1)
          (GameplayEffects|ApplyGameplayEffectToTarget :self _returnvalue :GameplayEffectClass "/GrillMoon/Gameplay/Effects/GE_GrillMoon_CombatAura.GE_GrillMoon_CombatAura_C" :Target _returnvalue :Level 0.0)
          (Utilities|Array|Add (Variables|Default|GetActors%20In%20Combat%20Mode) ActorEnteringCombatMode)
          (bind _asfort_player_pawn (Utilities|Casting|CastToFortPlayerPawn ActorEnteringCombatMode))
          (CallFunction|AutoEliminateDBNOPlayer _asfort_player_pawn)
          (Default|CallPlayerAdded _asfort_player_pawn))))))

(fn OnRep_ShowZone
  (Rendering|SetActorHiddenInGame (Variables|Default|GetShowZone)))

(fn RemoveAutoDBNOEliminationforPlayer
  (params (Player :type Fort%20Player%20Pawn%20Object%20Reference)))

(fn RemoveCombatMode
  (params (LeavingActor :type Actor%20Object%20Reference))
  (bind _returnvalue (Ability|GetFortAbilitySystemComponent LeavingActor))
  (GameplayEffects|RemoveActiveGameplayEffectbySourceEffect _returnvalue "/GrillMoon/Gameplay/Effects/GE_GrillMoon_CombatAura.GE_GrillMoon_CombatAura_C" _returnvalue -1)
  (GameplayEffects|ApplyGameplayEffectToTarget :self _returnvalue :GameplayEffectClass "/GrillMoon/Gameplay/Effects/GE_GrillMoon_EventAuraFX.GE_GrillMoon_EventAuraFX_C" :Target _returnvalue :Level 0.0)
  (GameplayEffects|ApplyGameplayEffectToTarget :self _returnvalue :GameplayEffectClass "/GrillMoon/Gameplay/Effects/GE_GrillMoon_EventDamageImmune.GE_GrillMoon_EventDamageImmune_C" :Target _returnvalue :Level 0.0)
  (bind _asfort_player_pawn_athena (Utilities|Casting|CastToFortPlayerPawnAthena LeavingActor))
  (FortPawn|SetGravityMultiplier _asfort_player_pawn_athena (ScalableFloat|GetValueatLevelwithContext (Variables|Default|GetGravity%20Scale%20Multiplier) 1.0))
  (CallFunction|RemoveEventItems _asfort_player_pawn_athena)
  (Utilities|Array|RemoveItem (Variables|Default|GetActors%20In%20Combat%20Mode) LeavingActor)
  (bind _asfort_player_pawn (Utilities|Casting|CastToFortPlayerPawn LeavingActor))
  (Default|CallPlayerRemoved _asfort_player_pawn))

(fn RemoveEventItems
  (params (LeavingPawn :type Fort%20Player%20Pawn%20Athena%20Object%20Reference))
  (bind _returnvalue (Fortnite|GetFortPlayerControllerfromActor LeavingPawn))
  (Networking|SwitchHasAuthority
    (:Authority
      (bind _items (InventoryManagement|GetItemsfromInventoryOwner _returnvalue))
      (for _array_element _items
        (for _array_element_2 (Variables|Default|GetEventItems)
          (if (== _array_element_2 (Item|GetItemDefinition _array_element))
            (Inventory|RemoveFortItemFromPlayer _returnvalue _array_element -1 true))))
      (Weapon|EquipWeaponActor LeavingPawn 0) ; Equip NULL to unequip and transition to anim set with hands down at sides
    )))

(fn SetZoneActive
  (params (Activate :type Boolean))
  ; Set Combat Zone Active
  (CallFunction|EventRouter-StartListening) ; listen for events to\n* add GAs to players\n* turn the event off
  (Collision|SetCollisionEnabled (Variables|Default|GetCombatZone) (select Activate "QueryOnly" "NoCollision"))
  (Variables|Default|SetShowZone (not Activate))
  (if (not Activate)
    (CallFunction|RemoveCombatModeFromAllPlayers)))

(fn ConstructionScript)
